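# Material function that supplies World Position Offset for Reality Sapling materials, combining wind sway with mask-driven radial scaling.
# MF_GrassWind-GameplayWPO, masked by vertex color red, is scaled by SaplingGlobalWindStrength.
# This adds to a push along the normalized object-to-vertex direction, sized by Scale_Amt and blended in by MaskInput.
# The noise-axis rotation, wiggle and storm wind-parameter branches do not connect to the output.
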MaterialFunction /RealitySeedGameplay/Environment/Assets_RealitySapling/Materials/MFC_SaplingWPO
graph {nodes [
  ActorPositionWS()
  CollectionParameter() {collection=/Game/Athena/Environments/Landscape/MPC/MPC_Landscape, parameterName=Ambient_WindMotion_Amount}
  n1 = Time()
  ConstantBiasScale(None=Sine(None=Multiply(A=@n1) {constB=0.5}))
  n2 = VertexColor()
  n3 = MaterialFunctionCall!(SinkIntoGroundDistance, WindFalloffMask=@n2.R, AfterHeightModification) {materialFunction=/Game/Athena/Apollo/Environments/Landscape/SmallFoliage/MasterMaterial/MaterialFunctions/MF_GrassWind-GameplayWPO}
  n4 = ScalarParameter() {defaultValue=0.35, parameterName=SaplingGlobalWindStrength}
  n5 = WorldPosition()
  n6 = ObjectPositionWS()
  n7 = Multiply(A=Normalize(VectorInput=Subtract(A=@n5.XYZ, B=@n6)), B=VectorParameter() {parameterName=Scale_Amt, defaultValue=<rgba 7 7 7 1>}.RGB)
  n8 = FunctionInput(Preview=Constant()) {inputName=MaskInput, inputType=FunctionInput_Scalar}
  FunctionOutput(None=Add(A=Multiply(A=@n3.Wind, B=@n4), B=LinearInterpolate(B=@n7, Alpha=@n8)))
  n9 = MaterialFunctionCall() {materialFunction=/Game/Athena/Environments/Landscape/MaterialFunctions/MF_StormMask}
  LinearInterpolate(
   A=CollectionParameter() {collection=&MPC_Landscape, parameterName=Grass_WindPower}
   B=CollectionParameter() {collection=&MPC_Landscape, parameterName=Grass_WindPower_Storm}
   Alpha=@n9.Result
  )
  LinearInterpolate()
  LinearInterpolate()
  n10 = Constant2Vector()
  n11 = MaterialFunctionCall(Float3=ActorPositionWS()) {materialFunction=&BreakOutFloat3Components}
  n12 = Add(A=Add(A=@n11.R, B=@n11.R), B=@n11.B)
  n13 = ScalarParameter() {defaultValue=1024, group=RandomColor, parameterName=StepSize}
  n14 = LinearInterpolate(
   A=ScalarParameter() {parameterName=MinOffset}
   B=ScalarParameter() {defaultValue=1, parameterName=MaxOffset}
   Alpha=Frac(None=Divide(A=@n12, B=@n13))
  )
  n15 = TextureSample(UVs=Panner(Coordinate=Add(A=@n10, B=@n14)) {speedX=0.1, speedY=0.1}) {texture=/Game/Characters/Player/Male/Medium/Bodies/M_MED_Viper/Textures/Glitch/T_SoftNoise}
  n16 = LinearInterpolate(A=Constant3Vector() {constant=<rgba 1 0 0 0>}, B=Constant3Vector() {constant=<rgba 0 2 0 0>}, Alpha=@n15.R)
  n17 = Multiply(
   A=Add(A=@n14, B=Time())
   B=LinearInterpolate(
    A=ScalarParameter() {defaultValue=0.05, parameterName=MacroSpeedMin}
    B=ScalarParameter() {defaultValue=0.1, parameterName=MacroSpeed}
    Alpha=@n14
   )
  ) {constB=0.25}
  n18 = Multiply(A=ConstantBiasScale(None=Sine(None=@n17))) {constB=0.5}
  n19 = Time()
  n20 = ScalarParameter() {defaultValue=0.1, parameterName=MedSpeedMin}
  n21 = ScalarParameter() {defaultValue=0.25, parameterName=MedSpeed}
  n22 = Sine(None=Multiply(A=Add(A=@n14, B=@n19), B=LinearInterpolate(A=@n20, B=@n21, Alpha=@n14)) {constB=0.5, desc="`"})
  n23 = Time()
  n24 = ScalarParameter() {defaultValue=0.25, parameterName=MicroSpeedMin}
  n25 = ScalarParameter() {defaultValue=0.5, parameterName=MicroSpeed}
  n26 = Sine(None=Multiply(A=Add(A=@n14, B=@n23), B=LinearInterpolate(A=@n24, B=@n25, Alpha=@n14)) {desc="`"})
  n27 = Add(
   A=Multiply(A=ConstantBiasScale(None=@n22), B=ScalarParameter() {defaultValue=0.5, parameterName=MicroStrength})
   B=Multiply(A=ConstantBiasScale(None=@n26), B=ScalarParameter() {defaultValue=0.1, parameterName=MicroStrength2})
  )
  n28 = Multiply(
   A=NamedRerouteUsage()
   B=LinearInterpolate(
    A=ScalarParameter() {defaultValue=-0.02, parameterName=MinRot}
    B=ScalarParameter() {defaultValue=0.02, parameterName=MaxRot}
    Alpha=Add(A=@n18, B=@n27)
   )
  )
  n29 = NamedRerouteUsage()
  n30 = Add(A=@n29, B=Multiply(A=Time(), B=ScalarParameter() {defaultValue=1.5, parameterName=WiggleSpeed}))
  n31 = ScalarParameter() {defaultValue=2, parameterName=WiggleFrequency}
  n32 = Multiply(A=Sine(None=Multiply(A=@n30, B=@n31) {constB=2, desc=1}), B=ScalarParameter() {defaultValue=0.0005, parameterName=WiggleIntensity}) {constB=0.02}
  n33 = RotateAboutAxis(
   NormalizedRotationAxis=@n16
   RotationAngle=@n32
   PivotPoint=MaterialFunctionCall() {materialFunction=&ObjectPivotPoint}.Object Pivot Location
   Position=WorldPosition().XYZ
  )
  n34 = ScalarParameter() {parameterName=WiggleControl}
  LinearInterpolate(
   A=RotateAboutAxis(
    NormalizedRotationAxis=@n16
    RotationAngle=Multiply(A=@n28, B=@n2.R)
    PivotPoint=MaterialFunctionCall() {materialFunction=&ObjectPivotPoint}.Object Pivot Location
    Position=WorldPosition().XYZ
   )
   B=@n33
   Alpha=@n34
  )
  LinearInterpolate(
   A=CollectionParameter() {collection=&MPC_Landscape, parameterName=Grass_WindSpeed}
   B=CollectionParameter() {collection=&MPC_Landscape, parameterName=Grass_WindSpeed_Storm}
   Alpha=@n9.Result
  )
  n35 = MaterialFunctionCall(Float3=@n15.RGB) {materialFunction=&BreakOutFloat3Components}
  MaterialFunctionCall(X=@n35.R, Y=@n35.R, Z=Constant()) {materialFunction=&MakeFloat3}
  Multiply(A=@n32, B=@n29)
  Multiply(A=@n33, B=@n34)
  n36 = Subtract(
   A=WorldPosition().XYZ
   B=MaterialFunctionCall() {materialFunction=&ObjectPivotPoint}.Object Pivot Location
  )
  n37 = MaterialFunctionCall() {materialFunction=&ObjectLocalBounds}
  NamedRerouteDeclaration(None=ChannelMaskParameter(None=Divide(A=@n36, B=@n37.Local Bounds Max)) {maskChannel=Blue, parameterName=Z-Gradient, defaultValue=<rgba 0 0 1 0>}.R) {name=ZGradient, nodeColor=<rgba 0 0.8594 1 1>}
  PerInstanceRandom()
  PerInstanceRandom()
  PerInstanceRandom()
 ]}
