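# Translucent, unlit, two-sided material for a glow sprite on the Reality Sapling, usable on particles and meshes.
# The Sprite function's camera-facing offset is added to MFC_SaplingWPO, masked by a SphereMask over a bounds-normalized height gradient shifted by ReactiveEmissive.
# A soft flare texture scaled by overall opacity and SaplingTeleport forms the glow's alpha, alongside a tintable bright cyan-white Color parameter.
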Material /RealitySeedGameplay/Environment/Assets_RealitySapling/FX/M_RealitySapling_GlowSprite
graph {
 properties {
  bEnableResponsiveAA=true
  bUsedWithMeshDeformer=true
  bUsedWithMeshParticles=true
  bUsedWithNiagaraMeshParticles=true
  bUsedWithNiagaraSprites=true
  bUsedWithParticleSprites=true
  bUsedWithSkeletalMesh=true
  blendMode=BLEND_Translucent
  opacityMaskClipValue=0.01
  pixelDepthOffsetMode=PDOM_Legacy
  shadingModel=MSM_Unlit
  twoSided=true
  usedShadingModels=Unlit
 }
 nodes [
  n1 = Subtract(
   A=WorldPosition().XYZ
   B=MaterialFunctionCall() {materialFunction=&ObjectPivotPoint}.Object Pivot Location
  )
  n2 = MaterialFunctionCall() {materialFunction=&ObjectLocalBounds}
  n3 = Subtract(
   A=ChannelMaskParameter(None=Divide(A=@n1, B=@n2.Local Bounds Max)) {maskChannel=Blue, parameterName=Z-Gradient, defaultValue=<rgba 0 0 1 0>}.R
   B=ScalarParameter() {defaultValue=-0.25, parameterName=ReactiveEmissive}
  )
  n4 = ScalarParameter() {parameterName=ReactiveEmissiveSamplePos}
  n5 = ScalarParameter() {defaultValue=0.25, parameterName=ReactiveEmissiveRadius}
  n6 = ScalarParameter() {parameterName=ReactiveEmissiveHardness}
  Add(
   A=MaterialFunctionCall!(
    Center Location=WorldPosition().XYZ
    0-1 UVs
    XY scale=ScalarParameter() {defaultValue=20, parameterName=SpriteSize}
    OverridePivot
    "Normals (Optional)"
   ) {materialFunction=&Sprite}.World position offset
   B=MaterialFunctionCall!(MaskInput=SphereMask(A=@n3, B=@n4, Radius=@n5, Hardness=@n6)) {materialFunction=/RealitySeedGameplay/Environment/Assets_RealitySapling/Materials/MFC_SaplingWPO}.Result
  )
  n7 = ScalarParameter() {defaultValue=1, parameterName=SaplingTeleport}
  Multiply(
   A=Multiply(
    A=TextureSampleParameter2D() {parameterName=FlareA, texture=/Game/Global/Textures/Default/Masks_and_Gradients/T_New_Soft_Flair}.R
    B=ScalarParameter() {defaultValue=0.15, parameterName=OveralOpacity}
   )
   B=@n7
  )
  OneMinus(None=@n7)
  VectorParameter() {parameterName=Color, defaultValue=<rgba 1.3075 1.8187 2 1>}
 ]
}
