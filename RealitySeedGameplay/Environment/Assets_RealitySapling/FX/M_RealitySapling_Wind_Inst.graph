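# Masked, lit material for Reality Sapling wind FX, usable on skeletal meshes and Niagara mesh particles.
# World position offset combines sine sway driven by UVs, vertex color and particle dynamic parameters with near-camera fade, a sapling WPO function, and hit/shrink deformation.
# A SaplingTeleport parameter sweeps a height-based noise dissolve with cyan glow and radial push.
# Gradient color, panned normals and optional erosion masking complete the surface.
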Material /RealitySeedGameplay/Environment/Assets_RealitySapling/FX/M_RealitySapling_Wind_Inst
graph {
 properties {
  bUsedWithMeshDeformer=true
  bUsedWithNiagaraMeshParticles=true
  bUsedWithSkeletalMesh=true
  blendMode=BLEND_Masked
  pixelDepthOffsetMode=PDOM_Legacy
  usedShadingModels=Default Lit
 }
 nodes [
  n1 = Divide(
   A=Subtract(
    A=WorldPosition().XYZ
    B=MaterialFunctionCall() {materialFunction=&ObjectPivotPoint}.Object Pivot Location
   )
   B=MaterialFunctionCall() {materialFunction=&ObjectLocalBounds}.Local Bounds Max
  )
  n2 = SphereMask(
   A=Subtract(
    A=ChannelMaskParameter(None=@n1) {maskChannel=Blue, parameterName=Z-Gradient, defaultValue=<rgba 0 0 1 0>}.R
    B=ScalarParameter() {defaultValue=-0.25, parameterName=ReactiveEmissive}
   )
   B=ScalarParameter() {parameterName=ReactiveEmissiveSamplePos}
   Radius=ScalarParameter() {defaultValue=0.25, parameterName=ReactiveEmissiveRadius}
   Hardness=ScalarParameter() {parameterName=ReactiveEmissiveHardness}
  )
  n3 = VectorParameter() {group=WPO, parameterName=WPO_Direction, defaultValue=<rgba 0.25 0.25 0 1>}
  n4 = TextureCoordinate() {coordinateIndex=1, uTiling=0, vTiling=0.7}
  n5 = ComponentMask(None=VectorParameter() {group=WPO, parameterName=WPO_UV_Scale, defaultValue=<rgba 1 1 0 1>}.RGB) {g=true, r=true}
  n6 = Add(A=ComponentMask(None=Multiply(A=@n4, B=@n5)) {g=true}, B=ScalarParameter() {group=WPO, parameterName=WPO_Offset})
  n7 = VertexColor()
  n8 = Add(
   A=Add(
    A=Add(A=@n6, B=@n7.G)
    B=DynamicParameter() {defaultValue=<rgba 1 1 0.5 0>, paramNames [RandomOffset, WPO_Intensity, WPO_Speed, Dissolve]}.RandomOffset
   )
   B=VertexColor().R
  )
  n9 = Multiply(A=Time(), B=DynamicParameter() {defaultValue=<rgba 1 1 0.5 0>, paramNames [RandomOffset, WPO_Intensity, WPO_Speed, Dissolve]}.WPO_Speed) {constB=2.8}
  n10 = Multiply(A=Sine(None=Add(A=@n8, B=@n9) {constB=0})) {constB=20}
  n11 = ComponentMask(None=TextureCoordinate() {coordinateIndex=1}) {g=true}
  n12 = LinearInterpolate(Alpha=Add(A=OneMinus(None=@n11)) {constB=0.5})
  n13 = Transform(None=Multiply(A=Multiply(A=@n3.RGB, B=@n10), B=Saturate(None=@n12))) {transformSourceType=TRANSFORMSOURCE_View}
  n14 = ComponentMask(None=TextureCoordinate() {coordinateIndex=1}) {g=true}
  n15 = Power(
   Base=Add(A=OneMinus(None=@n14), B=ScalarParameter() {group=WPO, parameterName=Bottom_Mask_Offset})
   Exp=ScalarParameter() {defaultValue=2.9, group=WPO, parameterName=Power}
  )
  n16 = ScalarParameter() {defaultValue=50, group=WPO, parameterName=Power_Mask_Edge}
  n17 = Multiply(
   A=Multiply(
    A=Multiply(A=@n13) {constB=0.5}
    B=DynamicParameter() {defaultValue=<rgba 1 1 0.5 0>, paramNames [RandomOffset, WPO_Intensity, WPO_Speed, Dissolve]}.WPO_Intensity
   )
   B=Saturate(None=Multiply(A=@n15, B=@n16) {constB=50})
  )
  n18 = ScalarParameter() {defaultValue=0.3, group=WPO, parameterName=Min_Near_Camera_WPO}
  n19 = MaterialFunctionCall(Distance=ScalarParameter() {defaultValue=256, group=WPO, parameterName=Near_Camera_fade}) {materialFunction=&Near_Camera_Fade}
  n20 = LinearInterpolate(
   B=Saturate(None=LinearInterpolate(A=@n18, Alpha=@n19.Result) {constA=0.5})
   Alpha=ScalarParameter() {group=WPO, parameterName=Use_Near_Camera_Fade}
  ) {constA=1}
  n21 = Add(
   A=LinearInterpolate(B=Multiply(A=@n17, B=@n20), Alpha=ScalarParameter() {defaultValue=0.5, group=WPO, parameterName=Use_WPO})
   B=Multiply(A=ScalarParameter() {group=WPO, parameterName=Small_WPO_Intensity}, B=NamedRerouteUsage())
  ) {constB=0}
  n22 = NamedRerouteUsage()
  Add(
   A=Add(
    A=MaterialFunctionCall!(MaskInput=@n2) {materialFunction=/RealitySeedGameplay/Environment/Assets_RealitySapling/Materials/MFC_SaplingWPO}.Result
    B=Add(A=@n21, B=@n22)
   )
   B=MaterialFunctionCall!(
    WSDeformationCenter=VectorParameter() {bUseCustomPrimitiveData=true, group=z_code, parameterName=WSDeformationCenter, primitiveDataIndex=20}.RGB
    Code_DestructionAmount=ScalarParameter() {group=z_code, parameterName=Code_DestructionAmount}
    Code_UseUvForDamage
    Code_UseHitEffects=StaticBoolParameter() {defaultValue=true, group=z_code, parameterName=Code_UseHitEffects}
    Code_Emissive
    WSDamageRadius=ScalarParameter() {bUseCustomPrimitiveData=true, defaultValue=1, group=z_code, parameterName=WSDamageRadius, primitiveDataIndex=26}
    CrackDamageShader
    Diffuse
    ScalarModDamageCracks
    ScalarOffsetforDamageCracks
    ExposeLerpFunctionForInvertingTheHitDistanceCalc
    InvertDamageFalloff
    CrackWorldPosition
    WorldPosition
    DistanceCullFadeAnimationOffset=ScalarParameter() {bUseCustomPrimitiveData=true, group=z_code, parameterName=DistanceCullFadeAnimationOffset, primitiveDataIndex=12}
   ) {materialFunction=/Game/Packages/jl_tests/Fortnite_HitEffects}.Hit And Shrink WPO
  )
  n23 = VectorParameter() {parameterName=Normal UV, defaultValue=<rgba 2 2 0 0>}
  n24 = AppendVector(A=Constant(), B=VertexColor().R)
  n25 = Add(A=Multiply(A=TextureCoordinate() {coordinateIndex=1, uTiling=1.5, vTiling=1.5}, B=ComponentMask(None=@n23.RGB) {g=true, r=true}), B=@n24)
  n26 = VectorParameter() {parameterName=Normal Pan Speed, defaultValue=<rgba 0 0.1 0 1>}
  n27 = NamedRerouteDeclaration(None=TextureSample(UVs=Panner(Coordinate=@n25, Speed=@n26.RGB) {speedY=2}) {samplerType=SAMPLERTYPE_Normal, texture=/Game/Characters/Player/Male/Large/Bodies/M_LRG_Clash_V/Textures/Simple_Perlin_N}.RGB) {name=Normal, nodeColor=<rgba 1 0.4922 0 1>}
  n28 = Add(A=@n24, B=TextureCoordinate() {uTiling=2, vTiling=2})
  n29 = LinearInterpolate(Alpha=TextureSample(UVs=Panner(Coordinate=@n28) {speedY=-0.3}) {texture=/Game/Effects/Fort_Effects/Textures/Power_Source/T_Power_Source_Sphere_Texture}.R) {constA=-0.2, constB=5}
  n30 = AppendVector(A=Multiply(A=ComponentMask(None=@n27) {g=true, r=true}, B=Saturate(None=@n29)), B=Constant() {r=1})
  n31 = ScalarParameter() {defaultValue=0.9, group=Normal, parameterName=Flatness}
  n32 = TextureCoordinate() {coordinateIndex=1}
  n33 = Multiply(A=Power(Base=ComponentMask(None=@n32) {g=true})) {constB=10}
  n34 = TextureSampleParameter2D(UVs=NamedRerouteUsage()) {
   group=Normal
   parameterName=Normal
   samplerType=SAMPLERTYPE_Normal
   texture=/Game/Athena/Apollo/Water/Textures/02_Rounded/02_WaterFoam_Veroni_base_N
  }
  n35 = TextureSampleParameterSubUV!(UVs, Apply View MipBias) {parameterName=Normal_SubUV, samplerType=SAMPLERTYPE_Normal, texture=/Game/Characters/Player/Male/Medium/Bodies/M_MED_Gumball/FX/T_Gumballs_N}
  LinearInterpolate(
   A=LinearInterpolate(
    A=Constant3Vector() {constant=<rgba 0 0 1 0>}
    B=MaterialFunctionCall(Normal=@n30, Flatness=@n31) {materialFunction=&FlattenNormal}.Result
    Alpha=Saturate(None=@n33)
   )
   B=LinearInterpolate(A=@n34.RGB, B=@n35.RGB, Alpha=ScalarParameter() {defaultValue=1, parameterName=Use_Sub_UV})
   Alpha=ScalarParameter() {group=Normal, parameterName=Use_Normal_Texture}
  )
  n36 = DynamicParameter() {defaultValue=<rgba 1 1 0.5 0>, paramNames [RandomOffset, WPO_Intensity, WPO_Speed, Dissolve]}
  n37 = AppendVector(A=@n36.RandomOffset, B=@n36.RandomOffset)
  n38 = DynamicParameter() {parameterIndex=1, paramNames [Erosion_U, Erosion_V, Brightness_in_game, Use_Frontend_bpr]}
  n39 = Multiply(A=TextureCoordinate() {coordinateIndex=1, uTiling=0.85, vTiling=0.85}, B=AppendVector(A=@n38.Erosion_U, B=@n38.Erosion_U))
  n40 = TextureSampleParameter2D(UVs=NamedRerouteDeclaration(None=Add(A=@n37, B=@n39)) {name=ErosionUVs, nodeColor=<rgba 1 1 1 1>}) {
   group=Mask
   parameterName=Mask
   samplerType=SAMPLERTYPE_LinearColor
   texture=/Game/Athena/Apollo/Water/Textures/02_Rounded/02_WaterFoam_Veroni_base_H
  }
  n41 = DynamicParameter() {defaultValue=<rgba 1 1 0.5 0>, paramNames [RandomOffset, WPO_Intensity, WPO_Speed, Dissolve]}
  Multiply(
   A=LinearInterpolate(A=Constant() {r=1}, B=Add(A=@n40.RGB, B=@n41.Dissolve) {constB=10}, Alpha=ScalarParameter() {group=Mask, parameterName=Use_Mask_Erocion})
   B=NamedRerouteUsage()
  )
  n42 = PreSkinnedPosition()
  n43 = ShaderStageSwitch(PixelShader=VertexInterpolator(VS=@n42).PS, VertexShader=@n42)
  n44 = ChannelMaskParameter(None=Divide(A=@n43, B=PreSkinnedLocalBounds().Max)) {maskChannel=Blue, parameterName="Z-CHanne;l", defaultValue=<rgba 0 0 1 0>}
  n45 = ScalarParameter() {defaultValue=1, parameterName=SaplingTeleport}
  n46 = Subtract(A=Add(A=OneMinus(None=@n44.R)), B=Multiply(A=OneMinus(None=@n45)) {constB=2})
  n47 = Time()
  n48 = VectorParameter() {parameterName=PanSpeedAndDirection, defaultValue=<rgba 0 0 64 1>}
  n49 = MaterialFunctionCall(
   TextureObject=TextureObject() {texture=/Game/Athena/SafeZone/Textures/T_JanusShield3_M}
   TextureSize=VectorParameter() {parameterName=CellNoiseSize, defaultValue=<rgba 90 90 90 1>}.RGB
   WorldPosition=Add(A=NamedRerouteUsage(), B=Multiply(A=@n47, B=@n48.RGB))
  ) {materialFunction=/Game/Weapons/WeaponSkins/Wraps/MaterialFunctions/WorldAlignedTexture}
  n50 = MaterialFunctionCall(
   Transition Phase=SphereMask(A=@n46, B=Constant() {r=0.5}, Radius=ScalarParameter() {defaultValue=0.5, parameterName=1, sliderMin=1}, Hardness=Constant())
   Height Texture=ChannelMaskParameter(None=@n49.XYZ Texture) {parameterName=NoiseChannel}.R
   Contrast=Constant()
  ) {materialFunction=&HeightLerp}
  n51 = Fresnel()
  Multiply(A=@n50.Alpha, B=Clamp(None=Power(Base=@n51)) {minDefault=0.5})
  NamedRerouteDeclaration(None=ScalarParameter() {defaultValue=0.5, parameterName=Specular}) {name=Specular, nodeColor=<rgba 0 0.2031 1 1>}
  NamedRerouteDeclaration(None=@n43) {name=PreskinnedPos, nodeColor=<rgba 0.6641 1 0 1>}
  NamedRerouteDeclaration(None=MaterialFunctionCall(
    Transition Phase=@n46
    Height Texture=ChannelMaskParameter(None=@n49.XYZ Texture) {maskChannel=Green, parameterName=OPacityNoiseChannel, defaultValue=<rgba 0 1 0 0>}.R
    Contrast=ScalarParameter() {defaultValue=1, parameterName=Contrastr}
   ) {materialFunction=&HeightLerp}.Alpha) {name=TeleportOpacityMask, nodeColor=<rgba 1 0 0.1406 1>}
  n52 = Multiply(A=Constant3Vector() {constant=<rgba 0 0.6106 1 0>}, B=@n50.Alpha)
  n53 = ScalarParameter() {defaultValue=4, parameterName=GlowStrength}
  NamedRerouteDeclaration(None=Power(Base=Multiply(A=@n52, B=@n53))) {name=Teleport Emissive, nodeColor=<rgba 0 0.9062 1 1>}
  n54 = Normalize(VectorInput=Subtract(A=WorldPosition().XYZ, B=ObjectPositionWS()))
  n55 = VectorParameter() {parameterName=Scale_AmtTele, defaultValue=<rgba 32 32 0 1>}
  n56 = Constant() {r=0.25}
  n57 = Constant() {r=0.5}
  n58 = Constant()
  NamedRerouteDeclaration(None=LinearInterpolate(B=Multiply(A=@n54, B=@n55.RGB), Alpha=SphereMask(A=@n46, B=@n56, Radius=@n57, Hardness=@n58))) {name=Teleport WPO, nodeColor=<rgba 1 0.8203 0 1>}
  NamedRerouteDeclaration(None=ScalarParameter() {parameterName=Metalic}) {name=Metalic, nodeColor=<rgba 0 0.8594 1 1>}
  NamedRerouteDeclaration(None=ScalarParameter() {defaultValue=0.35, parameterName=Roughness}) {name=Roughness, nodeColor=<rgba 0.2891 0 1 1>}
  n59 = Add(
   A=ComponentMask(None=TextureCoordinate() {uTiling=1.5, vTiling=1.5}) {g=true}
   B=ScalarParameter() {defaultValue=0.75, parameterName=ColorGradientOffset}
  ) {constB=0}
  n60 = ScalarParameter() {parameterName=ColorGradientContrast}
  n61 = NamedRerouteDeclaration(None=LinearInterpolate(
    A=VectorParameter() {group=Diffuse, parameterName=Color1, defaultValue=<rgba 4 4 4 0>}.RGB
    B=VectorParameter() {group=Diffuse, parameterName=Color2, defaultValue=<rgba 0 0.3726 1 1>}.RGB
    Alpha=MaterialFunctionCall(In=@n59, Contrast=@n60) {materialFunction=&CheapContrast}.Result
   )) {name=Color, nodeColor=<rgba 0.5469 1 0 1>}
  n62 = ScalarParameter() {defaultValue=0.2, parameterName=EmissiveMult}
  NamedRerouteDeclaration(None=Add(A=Multiply(A=@n61, B=@n62), B=NamedRerouteUsage())) {name=Emissive, nodeColor=<rgba 1 0.3984 0 1>}
  NamedRerouteUsage()
  NamedRerouteUsage()
  NamedRerouteUsage()
  NamedRerouteUsage()
  NamedRerouteUsage()
  OneMinus(None=@n50.Alpha)
  VectorParameter() {parameterName=vdscv}
 ]
}
