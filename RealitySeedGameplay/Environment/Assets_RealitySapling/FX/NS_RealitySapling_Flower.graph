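# Niagara effect for the Reality Sapling flower, layering butterflies, rotating mesh shards, rising sparkles, center sparks, mesh tendrils and a pulsing light.
# Butterflies spiral upward through vortex and line-attraction forces, while a second emitter reads their particles to spawn trailing sparks.
# User.Nurtuted boosts sparkle and spark spawn rates and speeds, plus light and tendril brightness.
# User.Opacity fades the layers, and System.NumPetals brightens the sparkles.
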NiagaraSystem /RealitySeedGameplay/Environment/Assets_RealitySapling/FX/NS_RealitySapling_Flower
dependencies {
 usedDataInterfaces [&NiagaraDataInterfaceCurve, &NiagaraDataInterfaceParticleRead]
 usedDynamicInputs []
 usedModules [
  &self:SystemSpawnScript.SystemScriptSource.SystemScriptGraph.NiagaraNodeAssignment_0.SetVariables_#1
  &SystemState
 ]
}
emitters {
 Butterfly001 {
  emitterUpdateScript [
   &EmitterState()
   &SpawnRate(SpawnRate=0.5, Use Spawn Probability=false)
  ]
  particleSpawnScript [
   &InitializeParticle(
    Lifetime=&RandomRangeFloat(Minimum=3, Maximum=4)
    Mesh Scale Mode=::NewEnumerator3
    Mesh Uniform Scale=0.06
    UsePositionOffset=false
    Write Material Random=true
   )
   &InitialMeshOrientation(Mesh Axis To Orient=<0 0 1>, Mesh Orientation Mode=::NewEnumerator2, Use Rotation Vector=false)
   &ShapeLocation(
    Cylinder Height=0
    MaskSpawnGroup=false
    Offset=<0 0 50>
    Offset Mode=::NewEnumerator0
    Override Seed=false
    Shape Primitive=::NewEnumerator1
    Surface Only=true
    Use Endcaps In Surface Only Mode=false
    UseOwnerScale=true
   )
   &AddVelocity(Velocity=<0 0 100>)
   /Niagara/Modules/Update/Velocity/VortexVelocity(Module.Use Influence Falloff=false, Velocity Amount=500) {enabled=false}
   &ApplyInitialForces()
  ]
  particleUpdateScript [
   &ParticleState()
   &VortexForce(Origin Pull Amount=120, UseOriginOffset=false, Vortex Force Amount=100)
   /Niagara/Modules/Update/Forces/LineAttractionForce(
    Attraction Strength=50
    Line End:Vector3f=&Add_Vector(A=Linked() {linkedVariable {name=Engine.Owner.Position, type=NiagaraPosition}}, B=<0 0 600>)
    Line Start:NiagaraPosition=$Engine.Owner.Position
    Use Attraction Falloff=false
   )
   &DynamicMaterialParameters(
    Index 0 Param 1=0.2364
    Index 0 Param 4=&Multiply_Float(A=$System.NormalizedLoopAge, B=3)
    Index 1 Param 1=3000
    Index 1 Param 2=1
    Module.Param0WriteEnabled=true
    Module.Param0WriteEnabled1=true
    Module.Param1WriteEnabled=true
    Module.Param1WriteEnabled1=true
    Module.Param2WriteEnabled=true
    Module.Param2WriteEnabled1=true
    Module.Param3WriteEnabled=true
    Module.Param3WriteEnabled1=true
    Write Parameter Index 1=true
   )
   &SolveForcesAndVelocity(Clamp Velocity=false, Limit Acceleration=false)
   &ScaleMeshSize(Scale Factor=&VectorFromFloat(Value=&FloatFromCurve(FloatCurve=DataInterface() {propertyValues {Curve {keys [
          <at 0 0 RCIM_Cubic>
          {arriveTangent=1.1765, interpMode=RCIM_Cubic, leaveTangent=1.1765, time=0.15, value=1}
          {arriveTangent=-1.1765, interpMode=RCIM_Cubic, leaveTangent=-1.1765, time=0.85, value=1}
          <at 1 0 RCIM_Cubic>
         ]}}})))
   &FN_Sprite_Attributes(
    Color=&MultiplyLinearColor(Linear Color=<rgba 0 0.5183 1 0.1>, Scale Factor=&MakeVector4(X=1, Y=1, Z=1, W=$User.Opacity))
    Create Color Attribute=true
    Create SpriteSize Attribute=true
    SpriteSize=&Multiply_Vector2D_ByFloat(Float=&MakeFloatFromVector(Vector=$Particles.Scale, Channel=::NewEnumerator0), Vector2D=<1200 1200>)
    Write to Module Namespace=true
   )
   &ScaleColor(Scale Alpha=$User.Opacity, ScaleA=true, ScaleRGB=false)
  ]
  rendererClasses [&NiagaraSpriteRendererProperties, &NiagaraMeshRendererProperties]
  renderers [
   {rendererClass=&NiagaraSpriteRendererProperties}
   {rendererClass=&NiagaraMeshRendererProperties, rendererIndex=1}
  ]
 }
 Butterfly_Sparks {
  emitterUpdateScript [
   &EmitterState()
   &SpawnParticlesFromOtherEmitter(
    Attribute Reader=DataInterface() {propertyValues {EmitterBinding {bindingMode=Other, emitterName=Butterfly001}}}
    Spawn Rate=20
    Use Maximum Age=false
    Use Minimum Age=false
   )
  ]
  particleSpawnScript [
   &InitializeParticle(
    AdjustAlpha=false
    AdjustHue=true
    AdjustSaturation=false
    AdjustValue=false
    Color=<rgba 0 1.1005 2 1>
    Color Mode=::NewEnumerator3
    Lifetime Max=1.5
    Lifetime Min=0.5
    Lifetime Mode=::NewEnumerator1
    Sprite Size Mode=::NewEnumerator1
    Uniform Sprite Size Max=13
    Uniform Sprite Size Min=4
    UsePositionOffset=false
    Write Material Random=true
   )
   &SampleParticlesFromOtherEmitter(
    Applied Position Offset:Vector3f=&RandomRangeVector(Minimum=<-8 -8 -8>, Maximum=<8 8 8>)
    Applied Velocity Scale=0.3
    Velocity Application:ENiagara_DirectReadApplicationMode=::NewEnumerator0
    Velocity Sampling=::NewEnumerator0
   )
   &ApplyInitialForces()
  ]
  particleUpdateScript [
   &ParticleState()
   &CurlNoiseForce(Noise Strength=100, Pan Noise=false)
   &Drag(Drag=3, Use Linear Drag=true, Use Rotational Drag=false)
   &GravityForce(Gravity=<0 0 -75>)
   &SolveForcesAndVelocity(Clamp Velocity=false, Limit Acceleration=false)
   &ScaleSpriteSize(Uniform Curve Sprite Scale=DataInterface() {propertyValues {Curve {keys [{interpMode=RCIM_Cubic, tangentMode=RCTM_User, time=0.3, value=1}, {interpMode=RCIM_Cubic, tangentMode=RCTM_User, time=1, value=0}]}}})
   &ScaleColor(Scale Alpha=$User.Opacity, ScaleA=true, ScaleRGB=false)
  ]
  rendererClasses [&NiagaraSpriteRendererProperties]
  renderers [{rendererClass=&NiagaraSpriteRendererProperties}]
 }
 Light {
  emitterUpdateScript [
   &EmitterState(Life Cycle Mode=::NewEnumerator1, Loop Duration=2, UseLoopDelay=false)
   &SpawnBurst_Instantaneous(Use Spawn Probability=false, UseLoopCountLimit=false)
  ]
  particleSpawnScript [&InitializeParticle(
    AdjustAlpha=false
    AdjustHue=true
    AdjustSaturation=true
    AdjustValue=false
    Color=<rgba 6.1519 0 22 1>
    Color Mode=::NewEnumerator3
    Hue Shift Range:Vector2f=<-0.013 0.026>
    Lifetime Max=3.6
    Lifetime Min=3
    Lifetime Mode=::NewEnumerator1
    Position Mode=::NewEnumerator0
    Saturation Range:Vector2f=<0.95 0.99>
    Sprite Size Mode=::NewEnumerator3
    Uniform Sprite Size=333
    Write Material Random=true
   )]
  particleUpdateScript [
   &ParticleState()
   &ScaleColor(
    Scale Alpha=$User.Opacity
    Scale RGB=&Add_Vector(
     A=&VectorFromFloat(Value=&FloatFromCurve(
       FloatCurve=DataInterface() {propertyValues {Curve {keys [
           <at 0 0.666 RCIM_Cubic>
           {arriveTangent=-0.8333, interpMode=RCIM_Cubic, leaveTangent=-0.8333, time=0.1, value=1}
           {arriveTangent=-0.875, interpMode=RCIM_Cubic, leaveTangent=-0.875, time=0.3, value=0.5}
           <at 0.5 0.6 RCIM_Cubic>
           {arriveTangent=0.724, interpMode=RCIM_Cubic, leaveTangent=0.724, time=0.6, value=0.4}
           {arriveTangent=0.75, interpMode=RCIM_Cubic, leaveTangent=0.75, time=0.7037, value=0.7475}
           {arriveTangent=-0.8354, interpMode=RCIM_Cubic, leaveTangent=-0.8354, time=0.8, value=0.55}
           <at 1 0.666 RCIM_Cubic>
          ]}}}
       CurveIndex=&Frac_Float(Float=&Multiply_Float(A=$Engine.Time, B=0.2))
       Scale Curve=&Multiply_Float(
        A=1
        B=&FloatFromCurve(FloatCurve=DataInterface() {propertyValues {Curve {keys [
             {arriveTangent=4.6341, interpMode=RCIM_Cubic, leaveTangent=4.6341, tangentMode=RCTM_User, time=0, value=0}
             {arriveTangent=-0.0405, interpMode=RCIM_Cubic, leaveTangent=-0.0405, tangentMode=RCTM_User, time=0.5, value=1}
             {interpMode=RCIM_Cubic, tangentMode=RCTM_User, time=0.666, value=1}
             {arriveTangent=-9.6003, interpMode=RCIM_Cubic, leaveTangent=-9.6003, tangentMode=RCTM_User, time=1, value=0}
            ]}}})
       )
      ))
     B=&VectorFromFloat(Value=&Multiply_Float(A=$User.Nurtuted, B=1))
    )
    ScaleA=true
    ScaleRGB=true
   )
   /SetVariables(
    Particles.LightOffsetPosition:NiagaraPosition=&Add_Vector(
     A=Linked() {linkedVariable {name=Particles.Position, type=NiagaraPosition}}
     B=&MakeVector(
      X=0
      Y=0
      Z=&FloatFromCurve(
       FloatCurve=DataInterface() {propertyValues {Curve {keys [{interpMode=RCIM_Cubic, tangentMode=RCTM_User, time=0, value=0}, {interpMode=RCIM_Cubic, tangentMode=RCTM_User, time=1, value=1}]}}}
       Scale Curve=77
      )
     )
    )
    Particles.LightRadius=&FloatFromCurve(
     FloatCurve=DataInterface() {propertyValues {Curve {keys [
         {arriveTangent=5.1313, interpMode=RCIM_Cubic, leaveTangent=5.1313, tangentMode=RCTM_User, time=0, value=0}
         {interpMode=RCIM_Cubic, tangentMode=RCTM_User, time=0.5, value=1}
         {arriveTangent=-1.2363, interpMode=RCIM_Cubic, tangentMode=RCTM_Break, time=1, value=0.75}
        ]}}}
     Scale Curve=100
    )
   ) {isSetParameters=true}
   &RotateAroundPoint(Radius=22, Rotation Rate:NiagaraFloat=&RandomRangeFloat(Minimum=-0.3, Maximum=0.4))
  ]
  rendererClasses [&NiagaraLightRendererProperties]
  renderers [{rendererClass=&NiagaraLightRendererProperties, rendererIndex=0}]
 }
 Shards {
  emitterUpdateScript [
   &EmitterState(Life Cycle Mode=::NewEnumerator1, Loop Behavior=::NewEnumerator1, Loop Duration Mode=::NewEnumerator1, UseLoopDelay=true)
   &SpawnBurst_Instantaneous(Spawn Count=15, Use Spawn Probability=false)
  ]
  particleSpawnScript [
   &FN_SetSpawnGroup()
   &InitializeParticle(
    Mesh Scale Mode=::NewEnumerator3
    Mesh Uniform Scale=/Game/Effects/Niagara/DynamicInputs/FN_PickFloatFromSpawnGroupCutoff(
     Over Cutoff Float:NiagaraFloat=&RandomRangeFloat(Minimum=0.3, Maximum=0.75)
     Under Cutoff Float:NiagaraFloat=&RandomRangeFloat(Minimum=1.5, Maximum=2)
     SpawnGroup:NiagaraInt32=$Particles.SpawnGroup
     Cutoff:NiagaraInt32=1
    )
    UsePositionOffset=false
    Write Material Random=true
   ) {module=InitializeParticle001}
   /SetVariables(Particles.VisibilityTag=&RandomRangeInt(Minimum=0, Maximum=2)) {isSetParameters=true}
   &InitialMeshOrientation(
    Orientation Axis=<0 1 0>
    Orientation Vector=<0 0 1>
    Rotation=&MakeVector(
     X=&RandomRangeFloat(Minimum=-0.05, Maximum=0.05)
     Y=HlslExpression() {hlslExpression="float(Particles.UniqueID) / 4"}
     Z=&RandomRangeFloat(Minimum=-0.05, Maximum=0.05)
    )
    Use Orientation Vector=true
    Use Rotation Vector=true
   )
  ]
  particleUpdateScript [
   &ParticleState(Kill Particles When Lifetime Has Elapsed=false, Let Infinitely Lived Particles Die When Emitter Deactivates=true)
   &UpdateMeshOrientation(
    Rotation Rate=&RandomRangeFloat(Minimum=0.1, Maximum=0.3)
    Rotation Vector=&RandomVector(Vector Scale=1)
   )
   &ShapeLocation(
    MaskSpawnGroup=false
    Offset=<0 0 375>
    Offset Mode=::NewEnumerator0
    Rotation Angle:NiagaraFloat=&Multiply_Float(A=$Emitter.Age, B=0.15)
    Rotation Axis=<1 0 1>
    Rotation Mode=::NewEnumerator1
    Sphere Distribution=::NewEnumerator3
    Sphere Radius=45
    UseOwnerScale=true
   )
   &ScaleMeshSize(Scale Factor=&VectorFromFloat(Value=&FloatFromCurve(
      FloatCurve=DataInterface() {propertyValues {Curve {keys [{interpMode=RCIM_Cubic, tangentMode=RCTM_User, time=0, value=0}, {interpMode=RCIM_Cubic, tangentMode=RCTM_User, time=1, value=1}]}}}
      CurveIndex=$System.Age
     )))
   &ScaleColor(Scale Alpha=$User.Opacity, ScaleA=true, ScaleRGB=false)
  ]
  rendererClasses [&NiagaraMeshRendererProperties]
  renderers [
   {rendererClass=&NiagaraMeshRendererProperties, rendererIndex=0}
   {rendererClass=&NiagaraMeshRendererProperties, rendererIndex=1}
   {rendererClass=&NiagaraMeshRendererProperties, rendererIndex=2}
  ]
 }
 Sparkles {
  emitterUpdateScript [
   &EmitterState()
   &SpawnRate(
    SpawnRate=&Add_Float(A=16, B=&Multiply_Float(A=$User.Nurtuted, B=20))
    Use Spawn Probability=false
   )
  ]
  particleSpawnScript [
   &InitializeParticle(
    Color Maximum=<rgba 0.2316 1.2823 2 1>
    Color Minimum=<rgba 0.8716 0.1327 1 1>
    Color Mode=::NewEnumerator2
    Lifetime Max=4
    Lifetime Min=2.6
    Lifetime Mode=::NewEnumerator1
    Mesh Scale Mode=::NewEnumerator4
    Position Mode=::NewEnumerator1
    Sprite Rotation Mode=::NewEnumerator2
    Sprite Size Mode=::NewEnumerator1
    Uniform Sprite Size Max=33
    Uniform Sprite Size Min=9
    UsePositionOffset=false
    Write Material Random=true
   )
   /SetVariables(Particles.Color=&MultiplyLinearColor(
     Linear Color=$Particles.Color
     Scale Factor=&MakeVector4_FromFloat(Float=&Add_Float(
       A=1
       B=&Multiply_Float(A=&MakeFloatFromInt(Int=$System.NumPetals), B=0.25)
      ))
    )) {isSetParameters=true}
   &ShapeLocation(
    Cylinder Height=44
    Cylinder Height Midpoint=0
    Cylinder Radius=133
    MaskSpawnGroup=false
    Override Seed=false
    Shape Primitive=::NewEnumerator1
    Surface Only=false
    UseOwnerScale=true
   )
   &AddVelocity(
    Velocity=&RandomRangeVector(Minimum=<0 0 25>, Maximum=<0 0 100>)
    Velocity Speed Scale=&Multiply_Float(A=$User.Nurtuted, B=1)
   )
  ]
  particleUpdateScript [
   &ParticleState()
   &DynamicMaterialParameters(
    Index 0 Param 1=1
    Index 0 Param 2=1
    Index 0 Param 3=0.5
    Module.Param0WriteEnabled=true
    Module.Param1WriteEnabled=true
    Module.Param2WriteEnabled=true
    Module.Param3WriteEnabled=true
   )
   &ScaleColor(
    Scale Alpha=&FloatFromCurve(
     FloatCurve=DataInterface() {propertyValues {Curve {keys [
         <at 0 0 RCIM_Cubic>
         {interpMode=RCIM_Cubic, tangentMode=RCTM_User, time=0.0856, value=0.9914}
         {arriveTangent=-2.3503, interpMode=RCIM_Cubic, leaveTangent=-2.3503, tangentMode=RCTM_User, time=0.8081, value=0.5934}
         <at 1 0>
        ]}}}
     Scale Curve=$User.Opacity
    )
    Scale RGB=&VectorFromFloat(Value=1)
    ScaleA=true
    ScaleRGB=true
   )
   &ScaleSpriteSize(Scale Factor=&Vector2DFromFloat(Value=&FloatFromCurve(FloatCurve=DataInterface() {propertyValues {Curve {keys [
          <at 0 0 RCIM_Cubic>
          {arriveTangent=0.836, interpMode=RCIM_Cubic, leaveTangent=0.836, time=0.2, value=1}
          {arriveTangent=-1.8455, interpMode=RCIM_Cubic, leaveTangent=-1.8455, tangentMode=RCTM_User, time=0.6507, value=0.544}
          <at 1 0>
         ]}}})))
   &GravityForce(Gravity=&Multiply_VectorByFloat(
     Vector=<1 1 77>
     Float=&FloatFromCurve(FloatCurve=DataInterface() {propertyValues {Curve {keys [
          {arriveTangent=-4.6763, interpMode=RCIM_Cubic, leaveTangent=-4.6763, tangentMode=RCTM_User, time=0, value=1}
          {interpMode=RCIM_Cubic, tangentMode=RCTM_User, time=1, value=-0.5}
         ]}}})
    ))
   &CurlNoiseForce(Noise Frequency=66, Noise Strength=11, Pan Noise=false)
   &SolveForcesAndVelocity(Clamp Velocity=false, Limit Acceleration=false)
  ]
  rendererClasses [&NiagaraSpriteRendererProperties]
  renderers [{rendererClass=&NiagaraSpriteRendererProperties}]
 }
 Sparks_Center {
  emitterUpdateScript [
   &EmitterState()
   &SpawnRate(
    SpawnRate=&Add_Float(A=10, B=&Multiply_Float(A=50, B=$User.Nurtuted))
    Use Spawn Probability=false
   )
  ]
  particleSpawnScript [
   &InitializeParticle(
    AdjustAlpha=false
    AdjustHue=true
    AdjustSaturation=false
    AdjustValue=true
    Color=<rgba 0 2.2011 4 1>
    Color Mode=::NewEnumerator3
    Lifetime Min=0.5
    Lifetime Mode=::NewEnumerator1
    Sprite Size Mode=::NewEnumerator1
    Uniform Sprite Size Max=30
    Uniform Sprite Size Min=10
    UsePositionOffset=false
    Value Range:Vector2f=&Vector2DFromFloat(Value=&Add_Float(A=1, B=&Multiply_Float(A=$User.Nurtuted, B=2)))
    Write Material Random=true
   )
   &ShapeLocation(MaskSpawnGroup=false, Offset=<0 0 250>, Offset Mode=::NewEnumerator0, Override Seed=false, Sphere Radius=40, UseOwnerScale=true)
   &AddVelocity(
    Cone Axis=<0 0 1>
    Module.UseConeFalloff=false
    Module.UseVelDistribution=false
    Override Seed=false
    Velocity Mode=::NewEnumerator2
    Velocity Speed=&Multiply_Float(
     A=&Multiply_Float(A=$User.Nurtuted, B=10)
     B=&RandomRangeFloat(Minimum=50, Maximum=75)
    )
   )
  ]
  particleUpdateScript [
   &ParticleState()
   &PointAttractionForce(
    Attraction Radius=200
    Attraction Strength=25
    Attractor Position:NiagaraPosition=&Add_Vector(A=/Niagara/DynamicInputs/Helpers/SimulationPosition(), B=<0 0 350>)
    Kill Within Radius=false
    Module.Use Falloff=true
    UseAttractorOffset=false
   )
   &CurlNoiseForce(Noise Strength=100, Pan Noise=false)
   &Drag(Drag=3, Use Linear Drag=true, Use Rotational Drag=false)
   &SolveForcesAndVelocity(Clamp Velocity=false, Limit Acceleration=false)
   &ScaleSpriteSize(Uniform Curve Sprite Scale=DataInterface() {propertyValues {Curve {keys [
        {arriveTangent=4.5394, interpMode=RCIM_Cubic, leaveTangent=4.5394, tangentMode=RCTM_User, time=0, value=0}
        {interpMode=RCIM_Cubic, tangentMode=RCTM_User, time=0.2, value=1}
        {tangentMode=RCTM_User, time=1, value=0}
       ]}}})
   &ScaleColor(
    Scale RGB=&VectorFromFloat(Value=&Add_Float(A=1, B=&Multiply_Float(A=$User.Nurtuted, B=5)))
    ScaleA=true
    ScaleRGB=true
   ) {enabled=false}
   &ScaleColor(Scale Alpha=$User.Opacity, ScaleA=true, ScaleRGB=false) {module=ScaleColor001}
  ]
  rendererClasses [&NiagaraSpriteRendererProperties]
  renderers [{rendererClass=&NiagaraSpriteRendererProperties}]
 }
 Tendrils {
  emitterUpdateScript [
   &EmitterState(Life Cycle Mode=::NewEnumerator1, Loop Duration=4, UseLoopDelay=false)
   &SpawnRate(SpawnRate=3, Use Spawn Probability=false)
  ]
  particleSpawnScript [
   &InitializeParticle(
    Color=<rgba 9 9 9 2>
    Lifetime=2.6
    Mesh Scale=&MakeVector(
     X=&RandomRangeFloat(Minimum=0.8, Maximum=1.333)
     Y=1
     Z=&RandomRangeFloat(Minimum=0.666, Maximum=1.666)
    )
    Mesh Scale Mode=::NewEnumerator4
    UsePositionOffset=false
    Write Material Random=true
   )
   &InitialMeshOrientation(
    Mesh Orientation Mode=::NewEnumerator7
    Rotation=&MakeVector(X=0, Y=0, Z=&RandomRangeFloat(Minimum=0, Maximum=1))
    Use Rotation Vector=true
   )
   &ShapeLocation(MaskSpawnGroup=false, Override Seed=false, Ring Radius=44, Shape Primitive=::NewEnumerator4, UseOwnerScale=true)
  ]
  particleUpdateScript [
   &ParticleState()
   &DynamicMaterialParameters(
    Index 0 Param 1=&RandomRangeFloat(Minimum=0, Maximum=33.333)
    Index 0 Param 2=&FloatFromCurve(
     FloatCurve=DataInterface() {propertyValues {Curve {keys [{interpMode=RCIM_Cubic, tangentMode=RCTM_User, time=0, value=1}, {interpMode=RCIM_Cubic, tangentMode=RCTM_User, time=1, value=0.5}]}}}
     Scale Curve=&Multiply_Float(A=1, B=&RandomRangeFloat(Minimum=0.2, Maximum=0.4))
    )
    Index 0 Param 4=&FloatFromCurve(
     FloatCurve=DataInterface() {propertyValues {Curve {keys [
         {interpMode=RCIM_Cubic, tangentMode=RCTM_User, time=0, value=0}
         {interpMode=RCIM_Cubic, tangentMode=RCTM_User, time=0.5, value=1}
         {tangentMode=RCTM_User, time=1, value=0}
        ]}}}
     Scale Curve=&Multiply_Float(A=1, B=&RandomRangeFloat(Minimum=0.5, Maximum=1))
    )
    Module.Param0WriteEnabled=true
    Module.Param1WriteEnabled=true
    Module.Param2WriteEnabled=false
    Module.Param3WriteEnabled=true
   )
   &ScaleColor(
    Scale Alpha=&FloatFromCurve(
     FloatCurve=DataInterface() {propertyValues {Curve {keys [
         {arriveTangent=6.8255, interpMode=RCIM_Cubic, leaveTangent=6.8255, tangentMode=RCTM_User, time=0, value=0}
         {interpMode=RCIM_Cubic, tangentMode=RCTM_User, time=0.333, value=1}
         {tangentMode=RCTM_User, time=1, value=0}
        ]}}}
     Scale Curve=$User.Opacity
    )
    Scale RGB=&VectorFromFloat(Value=&Add_Float(A=1, B=&Multiply_Float(A=$User.Nurtuted, B=2)))
    ScaleA=true
    ScaleRGB=true
   )
   &ScaleMeshSize(Scale Factor=&MakeVector(
     X=1
     Y=1
     Z=&FloatFromCurve(
      FloatCurve=DataInterface() {propertyValues {Curve {keys [
          {arriveTangent=4.1785, interpMode=RCIM_Cubic, leaveTangent=4.1785, tangentMode=RCTM_User, time=0, value=0}
          {arriveTangent=0.9995, interpMode=RCIM_Cubic, leaveTangent=0.9995, tangentMode=RCTM_User, time=0.2223, value=0.6825}
          {arriveTangent=-0.0405, interpMode=RCIM_Cubic, leaveTangent=-0.0405, tangentMode=RCTM_User, time=1, value=1}
         ]}}}
      Scale Curve=&Multiply_Float(A=1, B=0.95)
     )
    ))
  ]
  rendererClasses [&NiagaraMeshRendererProperties]
  renderers [{rendererClass=&NiagaraMeshRendererProperties, rendererIndex=0}]
 }
}
summary {
 systemName=NS_RealitySapling_Flower
 emitters [
  {emitterName=Shards, rendererClasses [&NiagaraMeshRendererProperties]}
  {emitterName=Butterfly001, rendererClasses [&NiagaraSpriteRendererProperties, &NiagaraMeshRendererProperties]}
  {emitterName=Butterfly_Sparks, rendererClasses [&NiagaraSpriteRendererProperties]}
  {emitterName=Sparks_Center, rendererClasses [&NiagaraSpriteRendererProperties]}
  {emitterName=Light, rendererClasses [&NiagaraLightRendererProperties]}
  {emitterName=Sparkles, rendererClasses [&NiagaraSpriteRendererProperties]}
  {emitterName=Tendrils, rendererClasses [&NiagaraMeshRendererProperties]}
 ]
 userVariables [
  User.FadeInDuration {type=NiagaraFloat, defaultValue=1, description=""}
  User.MaterialAngleFade {type=NiagaraFloat, defaultValue=0.5, description=""}
  User.MaterialMiddleFadeOpacity {type=NiagaraFloat, defaultValue=0.666, description=""}
  User.MaterialPanSpeed {type=NiagaraFloat, defaultValue=0.222, description=""}
  User.MaterialWaveLength {type=NiagaraFloat, defaultValue=0.8, description=""}
  User.MaterialWaveSpeed {type=NiagaraFloat, defaultValue=0.133, description=""}
  User.Nurtuted {type=NiagaraFloat, defaultValue=0, description=""}
  User.Opacity {type=NiagaraFloat, defaultValue=1, description=""}
 ]
}
