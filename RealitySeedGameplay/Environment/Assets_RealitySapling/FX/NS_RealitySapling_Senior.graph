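# Niagara effect for the senior Reality Sapling: a GPU sprite emitter of purple sparks spawned in a small sphere above the system origin.
# The User.Nurtuted float raises spawn rate, outward velocity and HSV value brightness, leaving a slow trickle at zero.
# An attractor above the spawn point, curl noise and drag steer the sparks, while a curve quickly grows then shrinks them.
# A nurture-driven ScaleColor module is present but disabled.
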NiagaraSystem /RealitySeedGameplay/Environment/Assets_RealitySapling/FX/NS_RealitySapling_Senior
dependencies {
 usedDataInterfaces [&NiagaraDataInterfaceCurve]
 usedDynamicInputs []
 usedModules [
  &self:SystemSpawnScript.SystemScriptSource.SystemScriptGraph.NiagaraNodeAssignment_0.SetVariables_#1
  &SystemState
 ]
}
emitters {Sparks_Center {
  simTarget=GPUComputeSim
  emitterUpdateScript [
   &EmitterState()
   &SpawnRate(
    SpawnRate=&Add_Float(A=8, B=&Multiply_Float(A=50, B=$User.Nurtuted))
    Use Spawn Probability=false
   )
  ]
  particleSpawnScript [
   &InitializeParticle(
    AdjustAlpha=false
    AdjustHue=true
    AdjustSaturation=false
    AdjustValue=true
    Color=<rgba 2.0136 0.5582 4 1>
    Color Mode=::NewEnumerator3
    Lifetime Min=0.5
    Lifetime Mode=::NewEnumerator1
    Sprite Size Mode=::NewEnumerator1
    Uniform Sprite Size Max=15
    Uniform Sprite Size Min=3
    UsePositionOffset=false
    Value Range:Vector2f=&Vector2DFromFloat(Value=&Add_Float(A=1, B=&Multiply_Float(A=$User.Nurtuted, B=5)))
    Write Material Random=true
   )
   &ShapeLocation(MaskSpawnGroup=false, Offset=<0 0 250>, Offset Mode=::NewEnumerator0, Override Seed=false, Sphere Radius=20, UseOwnerScale=true)
   &AddVelocity(
    Module.UseConeFalloff=false
    Module.UseVelDistribution=false
    Override Seed=false
    Velocity Mode=::NewEnumerator2
    Velocity Speed=&Multiply_Float(A=$User.Nurtuted, B=&RandomRangeFloat(Minimum=50, Maximum=75))
   )
  ]
  particleUpdateScript [
   &ParticleState()
   &PointAttractionForce(
    Attraction Radius=200
    Attraction Strength=25
    Attractor Position:NiagaraPosition=&Add_Vector(A=/Niagara/DynamicInputs/Helpers/SimulationPosition(), B=<0 0 350>)
    Kill Within Radius=false
    Module.Use Falloff=true
    UseAttractorOffset=false
   )
   &CurlNoiseForce(Noise Strength=100, Pan Noise=false)
   &Drag(Drag=3, Use Linear Drag=true, Use Rotational Drag=false)
   &SolveForcesAndVelocity(Clamp Velocity=false, Limit Acceleration=false)
   &ScaleSpriteSize(Uniform Curve Sprite Scale=DataInterface() {propertyValues {Curve {keys [
        {arriveTangent=4.5394, interpMode=RCIM_Cubic, leaveTangent=4.5394, tangentMode=RCTM_User, time=0, value=0}
        {interpMode=RCIM_Cubic, tangentMode=RCTM_User, time=0.2, value=1}
        {tangentMode=RCTM_User, time=1, value=0}
       ]}}})
   &ScaleColor(
    Scale RGB=&VectorFromFloat(Value=&Add_Float(A=1, B=&Multiply_Float(A=$User.Nurtuted, B=5)))
    ScaleA=true
    ScaleRGB=true
   ) {enabled=false}
  ]
  rendererClasses [&NiagaraSpriteRendererProperties]
  renderers [{rendererClass=&NiagaraSpriteRendererProperties}]
 }}
summary {
 systemName=NS_RealitySapling_Senior
 emitters [{emitterName=Sparks_Center, simTarget=GPUComputeSim, rendererClasses [&NiagaraSpriteRendererProperties]}]
 userVariables [User.Nurtuted {type=NiagaraFloat, defaultValue=0, description=""}]
}
