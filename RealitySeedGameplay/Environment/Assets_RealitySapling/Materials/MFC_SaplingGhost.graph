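# Material function that adds a switchable ghost effect to Reality Sapling material attributes, overriding opacity, emissive and World Position Offset.
# Fresnel masks combine with panning soft noise and noise-distorted tendrils, projected from pre-skinned position, to drive opacity and power-boosted emissive tinted by a color lookup texture.
# Named reroutes supply camera-distance fade, inverted Z gradient and a WPO that collapses the mesh to its origin beyond fade range.
# Depth fade softens intersections.
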MaterialFunction /RealitySeedGameplay/Environment/Assets_RealitySapling/Materials/MFC_SaplingGhost
graph {nodes [
  Constant()
  Constant() {r=1}
  n1 = GetMaterialAttributes(None=FunctionInput() {inputName=Material, inputType=FunctionInput_MaterialAttributes})
  n2 = ScalarParameter() {parameterName=OPacityMin}
  n3 = Fresnel(ExponentIn=ScalarParameter() {defaultValue=0.25, parameterName=FresnelExponent}, Normal=PixelNormalWS())
  n4 = ScalarParameter() {defaultValue=1.5, parameterName=FresnelPower}
  n5 = PreSkinnedPosition()
  n6 = ShaderStageSwitch(PixelShader=VertexInterpolator(VS=@n5).PS, VertexShader=@n5)
  n7 = Time()
  n8 = VectorParameter() {parameterName=SoftNoisePanSpeed, defaultValue=<rgba 12 -12 -16 1>}
  n9 = MaterialFunctionCall(
   TextureObject=TextureObject() {texture=/Game/Characters/Player/Male/Medium/Bodies/M_MED_Viper/Textures/Glitch/T_SoftNoise}
   TextureSize=VectorParameter() {parameterName=SoftNoiseTiling, defaultValue=<rgba 90 90 90 1>}.RGB
   WorldPosition=Add(A=@n6, B=Multiply(A=@n7, B=@n8.RGB))
  ) {materialFunction=/Game/Weapons/WeaponSkins/Wraps/MaterialFunctions/WorldAlignedTexture}
  n10 = ChannelMaskParameter(None=@n9.XYZ Texture) {parameterName=Red}
  n11 = ScalarParameter() {defaultValue=0.5, parameterName=SoftnoiseStrength}
  n12 = Clamp(None=Subtract(A=Power(Base=@n3, Exp=@n4), B=Multiply(A=@n10.R, B=@n11)))
  n13 = Fresnel(ExponentIn=ScalarParameter() {defaultValue=0.2, parameterName=FresnelExponent2}, Normal=PixelNormalWS())
  n14 = Multiply(A=ComponentMask(None=@n9.XYZ Texture) {b=true, g=true, r=true}) {constB=2}
  n15 = Add(A=Multiply(A=Subtract(A=@n14), B=ScalarParameter() {defaultValue=4, parameterName=DistortionStrength}), B=@n6)
  n16 = Multiply(A=Time(), B=VectorParameter() {parameterName=TendrilPanSpeed, defaultValue=<rgba 0 0 -10 1>}.RGB)
  n17 = ComponentMask(None=MaterialFunctionCall(
    TextureObject=TextureObject() {samplerType=SAMPLERTYPE_LinearGrayscale, texture=/RealitySeedGameplay/Environment/Assets_RealitySapling/Textures/T_Tendrils_Blurred}
    TextureSize=VectorParameter() {parameterName=TendrilsTiling, defaultValue=<rgba 64 64 64 1>}.RGB
    WorldPosition=Add(A=@n15, B=@n16)
   ) {materialFunction=&WorldAlignedTexture}.XY Texture) {r=true}
  n18 = Multiply(A=@n17, B=ScalarParameter() {defaultValue=3, parameterName=GhostEmissiveNoiseStrength})
  n19 = Multiply(A=Power(Base=OneMinus(None=@n13), Exp=ScalarParameter() {defaultValue=2, parameterName=FresnelPower2}), B=@n18)
  n20 = Clamp(None=Multiply(A=Max(A=@n12, B=@n19), B=ScalarParameter() {defaultValue=1, parameterName=MasterOpacity}))
  n21 = NamedRerouteUsage()
  n22 = DepthFade(
   Opacity=Multiply(A=LinearInterpolate(A=@n2, B=@n20, Alpha=@n21), B=NamedRerouteUsage())
   FadeDistance=ScalarParameter() {defaultValue=64, parameterName=FadeDistance}
  )
  n23 = Subtract(A=@n12, B=Multiply(A=OneMinus(None=@n17), B=ScalarParameter() {defaultValue=0.25, parameterName=TendrilNoiseStrength}))
  n24 = Clamp(None=Max(A=Clamp(None=@n23), B=@n18))
  n25 = TextureSample(UVs=ChannelMaskParameter(None=@n9.XYZ Texture) {maskChannel=Green, parameterName=Green, defaultValue=<rgba 0 1 0 0>}.R) {texture=/Game/Environments/Artemis/Props/Daisy/Daisy_RealityGlitch/Textures/T_4colors}
  n26 = ScalarParameter() {defaultValue=4, parameterName=GhostEmissiveStrength_Min}
  n27 = ScalarParameter() {defaultValue=4, parameterName=GhostEmissiveStrength}
  n28 = NamedRerouteUsage()
  n29 = Power(
   Base=Multiply(A=Multiply(A=@n24, B=@n25.RGB), B=LinearInterpolate(A=@n26, B=@n27, Alpha=@n28))
   Exp=ScalarParameter() {defaultValue=2, parameterName=GhostEmissivePower}
  )
  n30 = NamedRerouteUsage()
  FunctionOutput(None=StaticSwitchParameter(
    True=SetMaterialAttributes!(MaterialAttributes=@n1.MaterialAttributes, Opacity=@n22, EmissiveColor=@n29, WorldPositionOffset=@n30) {attributeSetTypes [#1, #2, #3]}
    False=@n1.MaterialAttributes
   ) {parameterName=EnableSaplingGhost})
  n31 = ObjectPositionWS()
  n32 = WorldPosition()
  n33 = Distance(A=CameraPositionWS(), B=ObjectPositionWS())
  n34 = ScalarParameter() {defaultValue=4096, parameterName=CameraDistanceBlend}
  n35 = Clamp(None=OneMinus(None=Divide(A=@n33, B=@n34) {constB=2048}))
  NamedRerouteDeclaration(None=LinearInterpolate(A=Subtract(A=@n31, B=@n32.XYZ), Alpha=Ceil(None=@n35)) {constB=0}) {name=WPO, nodeColor=<rgba 0 0.8125 1 1>}
  NamedRerouteDeclaration(None=@n35) {name=CameraFade, nodeColor=<rgba 1 0.6094 0 1>}
  n36 = WorldPosition()
  n37 = MaterialFunctionCall() {materialFunction=&ObjectPivotPoint}
  n38 = ChannelMaskParameter(None=Divide(
    A=Subtract(A=@n36.XYZ, B=@n37.Object Pivot Location)
    B=MaterialFunctionCall() {materialFunction=&ObjectLocalBounds}.Local Bounds Max
   )) {maskChannel=Blue, parameterName=Z-Gradient, defaultValue=<rgba 0 0 1 0>}
  NamedRerouteDeclaration(None=OneMinus(None=FunctionInput(Preview=@n38.R) {bUsePreviewValueAsDefault=true, inputName=Z-Gradient, inputType=FunctionInput_Scalar})) {name=Z-Gradient, nodeColor=<rgba 1 0.8906 0 1>}
  TextureObject() {texture=/Game/Athena/Environments/Nexus/Textures/T_ZeroPoint_Line_Stars}
  n39 = TextureCoordinate()
  n40 = ScalarParameter() {defaultValue=2, parameterName=GhostNoiseTiling}
  TextureSample(UVs=Panner(Coordinate=Multiply(A=@n39, B=@n40)) {speedY=0.25}) {texture=&T_SoftNoise}
  VectorParameter() {parameterName=ReactiveEmissiveColor, defaultValue=<rgba 1 0.5846 0 1>}
 ]}
