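# Material function that adds a teleport dissolve to the Reality Sapling, driven by the SaplingTeleport parameter.
# A gradient from normalized pre-skinned height sweeps the mesh, broken up by panning WorldAlignedTexture noise fed with pre-skinned position through HeightLerp.
# It drives opacity and mask, a blue emissive band, and vertex offset away from the object center, horizontal by default.
# Static switch EnableOpacityMask gates both the attributes and separate WPO outputs.
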MaterialFunction /RealitySeedGameplay/Environment/Assets_RealitySapling/Materials/MFC_SaplingTeleport
graph {nodes [
  n1 = PreSkinnedPosition()
  n2 = ShaderStageSwitch(PixelShader=VertexInterpolator(VS=@n1).PS, VertexShader=@n1)
  n3 = ChannelMaskParameter(None=Divide(A=@n2, B=PreSkinnedLocalBounds().Max)) {maskChannel=Blue, parameterName="Z-CHanne;l", defaultValue=<rgba 0 0 1 0>}
  n4 = ScalarParameter() {defaultValue=1, parameterName=SaplingTeleport}
  n5 = Subtract(A=Add(A=OneMinus(None=@n3.R)), B=Multiply(A=OneMinus(None=@n4)) {constB=2})
  Clamp(None=@n5)
  n6 = Multiply(A=Time()) {constB=0.25}
  ConstantBiasScale(None=Sine(None=@n6))
  Frac(None=@n6)
  n7 = GetMaterialAttributes(None=FunctionInput() {inputName=Material, inputType=FunctionInput_MaterialAttributes}) {attributeGetTypes [#1, #2, #3]}
  n8 = Time()
  n9 = VectorParameter() {parameterName=PanSpeedAndDirection, defaultValue=<rgba 0 0 64 1>}
  n10 = MaterialFunctionCall(
   TextureObject=TextureObject() {texture=/Game/Athena/SafeZone/Textures/T_JanusShield3_M}
   TextureSize=VectorParameter() {parameterName=CellNoiseSize, defaultValue=<rgba 90 90 90 1>}.RGB
   WorldPosition=Add(A=NamedRerouteUsage(), B=Multiply(A=@n8, B=@n9.RGB))
  ) {materialFunction=/Game/Weapons/WeaponSkins/Wraps/MaterialFunctions/WorldAlignedTexture}
  n11 = MaterialFunctionCall(
   Transition Phase=SphereMask(A=@n5, B=Constant() {r=0.5}, Radius=ScalarParameter() {defaultValue=0.5, parameterName=1, sliderMin=1}, Hardness=Constant())
   Height Texture=ChannelMaskParameter(None=@n10.XYZ Texture) {parameterName=NoiseChannel}.R
   Contrast=Constant()
  ) {materialFunction=&HeightLerp}
  n12 = Multiply(
   A=Multiply(A=Constant3Vector() {constant=<rgba 0 0.6106 1 0>}, B=@n11.Alpha)
   B=ScalarParameter() {defaultValue=4, parameterName=GlowStrength}
  )
  n13 = Normalize(VectorInput=Subtract(A=WorldPosition().XYZ, B=ObjectPositionWS()))
  n14 = VectorParameter() {parameterName=Scale_AmtTele, defaultValue=<rgba 32 32 0 1>}
  n15 = Constant() {r=0.25}
  n16 = Constant() {r=0.5}
  n17 = Constant()
  n18 = Add(A=@n7.WorldPositionOffset, B=LinearInterpolate(B=Multiply(A=@n13, B=@n14.RGB), Alpha=SphereMask(A=@n5, B=@n15, Radius=@n16, Hardness=@n17)))
  n19 = MaterialFunctionCall(
   Transition Phase=@n5
   Height Texture=ChannelMaskParameter(None=@n10.XYZ Texture) {maskChannel=Green, parameterName=OPacityNoiseChannel, defaultValue=<rgba 0 1 0 0>}.R
   Contrast=ScalarParameter() {defaultValue=1, parameterName=Contrastr}
  ) {materialFunction=&HeightLerp}
  n20 = SetMaterialAttributes!(
   MaterialAttributes=@n7.MaterialAttributes
   EmissiveColor=Add(A=@n7.EmissiveColor, B=Power(Base=@n12))
   OpacityMask=@n19.Alpha
   WorldPositionOffset=@n18
   Opacity=Multiply(A=@n7.Opacity, B=@n19.Alpha)
  ) {attributeSetTypes [
    #1
    #4
    #2
    #3
   ]}
  FunctionOutput(None=StaticSwitchParameter(True=@n20, False=@n7.MaterialAttributes) {parameterName=EnableOpacityMask})
  FunctionOutput(None=StaticSwitchParameter(True=@n18, False=@n7.WorldPositionOffset) {parameterName=EnableOpacityMask}) {outputName=WPOResult}
  n21 = Fresnel()
  Multiply(A=@n11.Alpha, B=Clamp(None=Power(Base=@n21)) {minDefault=0.5})
  NamedRerouteDeclaration(None=@n2) {name=PreskinnedPos, nodeColor=<rgba 0.6641 1 0 1>}
  OneMinus(None=@n11.Alpha)
 ]}
