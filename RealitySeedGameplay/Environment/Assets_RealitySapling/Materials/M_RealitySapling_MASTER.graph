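# Master material for the Reality Sapling, built on a Fortnite-style layered surface with optional two-sided textures, tint masking and masked emissive.
# A parameter-positioned sphere mask on a height-gradient input drives noisy, multicolored reactive glow and sapling vertex offset.
# The surface passes through Ghost and Teleport functions, then height-blends into landscape runtime virtual textures near the ground.
# Snow progression and game-wide functions follow, with sapling offsets added last.
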Material /RealitySeedGameplay/Environment/Assets_RealitySapling/Materials/M_RealitySapling_MASTER
graph {
 properties {bUseMaterialAttributes=true, pixelDepthOffsetMode=PDOM_Legacy, usedShadingModels=Default Lit}
 nodes [
  ActorPositionWS()
  CollectionParameter() {collection=/Game/Athena/Environments/Landscape/MPC/MPC_Landscape, parameterName=Ambient_WindMotion_Amount}
  n1 = NamedRerouteUsage()
  ConstantBiasScale(None=Sine(None=Multiply(A=@n1) {constB=2}))
  Constant() {r=8}
  Divide(
   A=Subtract(
    A=WorldPosition().XYZ
    B=MaterialFunctionCall() {materialFunction=&ObjectPivotPoint}.Object Pivot Location
   )
   B=MaterialFunctionCall() {materialFunction=&ObjectLocalBounds}.Local Bounds Max
  )
  LinearInterpolate(Alpha=Frac(None=Time())) {constA=-0.25, constB=1.25}
  n2 = MaterialFunctionCall() {materialFunction=/Game/Athena/Environments/Landscape/MaterialFunctions/MF_StormMask}
  LinearInterpolate(
   A=CollectionParameter() {collection=&MPC_Landscape, parameterName=Grass_WindSpeed}
   B=CollectionParameter() {collection=&MPC_Landscape, parameterName=Grass_WindSpeed_Storm}
   Alpha=@n2.Result
  )
  LinearInterpolate(
   A=CollectionParameter() {collection=&MPC_Landscape, parameterName=Grass_WindPower}
   B=CollectionParameter() {collection=&MPC_Landscape, parameterName=Grass_WindPower_Storm}
   Alpha=@n2.Result
  )
  n3 = PreSkinnedPosition()
  n4 = ShaderStageSwitch(PixelShader=VertexInterpolator(VS=@n3).PS, VertexShader=@n3)
  n5 = PreSkinnedLocalBounds()
  NamedRerouteDeclaration(None=ChannelMaskParameter(None=Divide(A=@n4, B=@n5.Max)) {maskChannel=Blue, parameterName=Z-Gradient, defaultValue=<rgba 0 0 1 0>}.R) {name=ZGradient, nodeColor=<rgba 0 0.8594 1 1>}
  PerInstanceRandom()
  n6 = StaticBool() {value=true}
  n7 = MaterialFunctionCall() {materialFunction=/Game/Athena/WinterArt/Materials/MF_UVsSet}
  n8 = VectorParameter() {
   group=Material_1
   parameterName=L1_UVScaleByComponent
   sortPriority=3
   channelNames {a="", b=Normals, g=Specular, r=Diffuse}
   defaultValue=<rgba 1 1 1 0>
  }
  n9 = Multiply(A=@n7.UVs Layer 1, B=@n8.Specular)
  n10 = TextureSampleParameter2D(UVs=@n9) {
   group=Material_1
   parameterName=SpecularMasks Side 2
   samplerSource=SSM_Wrap_WorldGroupSettings
   samplerType=SAMPLERTYPE_Masks
   sortPriority=2
   texture=&T_Fortnite_Default_S
  }
  n11 = TextureCoordinate() {coordinateIndex=1}
  n12 = TextureSampleParameter2D(UVs=StaticSwitchParameter(
    True=TextureCoordinate() {coordinateIndex=2}
    False=StaticSwitchParameter(True=@n11, False=@n9) {group=Material_1, parameterName=L1_Ch2Emissive Side2, sortPriority=-5}
   ) {group=Material_1, parameterName=L1_Ch3Emissive Side2, sortPriority=-5}) {
   group=Material_1
   parameterName=L1_Emissive Side 2
   samplerSource=SSM_Wrap_WorldGroupSettings
   samplerType=SAMPLERTYPE_LinearGrayscale
   sortPriority=-4
   texture=/Game/Global/Textures/Default/Blanks/T_BlackLinearGrayscale
  }
  n13 = StaticSwitchParameter(True=@n12.R, False=@n10.A) {group=Material_1, parameterName=L1_CustomEmissive Side 2, sortPriority=-5}
  n14 = TwoSidedSign()
  n15 = VectorParameter() {
   group=Material_1
   parameterName=L1_UVScaleByComponent
   sortPriority=3
   channelNames {a="", b=Normals, g=Specular, r=Diffuse}
   defaultValue=<rgba 1 1 1 0>
  }
  n16 = Multiply(A=@n15.Specular, B=@n7.UVs Layer 1)
  n17 = TextureSampleParameter2D(UVs=@n16) {
   group=Material_1
   parameterName=SpecularMasks
   samplerSource=SSM_Wrap_WorldGroupSettings
   samplerType=SAMPLERTYPE_Masks
   sortPriority=-5
   texture=&T_Fortnite_Default_S
  }
  n18 = TextureCoordinate() {coordinateIndex=1}
  n19 = TextureSampleParameter2D(UVs=StaticSwitchParameter(
    True=TextureCoordinate() {coordinateIndex=2}
    False=StaticSwitchParameter(True=@n18, False=@n16) {group=Material_1, parameterName=L1_Ch2Emissive, sortPriority=-5}
   ) {group=Material_1, parameterName=L1_Ch3Emissive, sortPriority=1}) {
   group=Material_1
   parameterName=L1_Emissive
   samplerSource=SSM_Wrap_WorldGroupSettings
   samplerType=SAMPLERTYPE_LinearGrayscale
   sortPriority=-5
   texture=&T_BlackLinearGrayscale
  }
  n20 = AppendVector(A=@n17.RGB, B=StaticSwitchParameter(True=@n19.R, False=@n17.A) {group=Material_1, parameterName=L1_CustomEmissive, sortPriority=-5})
  n21 = StaticSwitchParameter(True=LinearInterpolate(A=AppendVector(A=@n10.RGB, B=@n13), B=@n20, Alpha=@n14), False=@n20) {group=Material_1, parameterName=TwoSided, sortPriority=18}
  n22 = MaterialFunctionCall(Float3=@n21) {materialFunction=&BreakOutFloat3Components}
  n23 = ChannelMaskParameter(None=@n21) {maskChannel=Alpha, parameterName="EmissiveChanne;", defaultValue=<rgba 0 0 0 1>}
  n24 = VectorParameter() {group=Material_1, parameterName=Emissive, sortPriority=7, defaultValue=<rgba 0 0 0 1>}
  n25 = MaterialFunctionCall!(
   X=ScalarParameter() {group=Adv Emissive, parameterName=L1 X Speed, sortPriority=12}
   Y=ScalarParameter() {group=Adv Emissive, parameterName=L1 Y Speed, sortPriority=13}
   Dot=VectorParameter() {group=Adv Emissive, parameterName=L1 Channel, sortPriority=9}.RGB
   TextureOffset=ScalarParameter() {defaultValue=1, group=Adv Emissive, parameterName=L1 Offset, sliderMax=4, sliderMin=1, sortPriority=11}
   EmissiveTexture=Multiply(A=@n24.RGB, B=@n24.RGB)
  ) {materialFunction=/Game/Athena/WinterArt/Materials/MF_EmissiveMask}
  n26 = TextureSampleParameter2D(UVs=Multiply(A=@n7.UVs Layer 1, B=@n8.Diffuse)) {
   group=Material_1
   parameterName=Diffuse Side 2
   samplerSource=SSM_Wrap_WorldGroupSettings
   sortPriority=1
   texture=/Game/Global/Textures/Default/Blanks/T_MidgreysRGB
  }
  n27 = Multiply(A=@n7.UVs Layer 1, B=@n15.Diffuse)
  n28 = StaticSwitchParameter(True=Multiply(A=ScalarParameter() {defaultValue=1, group=Material_1, parameterName=L1_UV_ColorSwatchScaleAdd}, B=@n27), False=@n27) {group=Material_1, parameterName=Use Color Swatch}
  n29 = TextureSampleParameter2D(UVs=Multiply(A=ScalarParameter() {defaultValue=1, group=Material_1, parameterName=L1_CustomHeightScale, sortPriority=1}, B=@n28)) {
   group=Material_1
   parameterName=L1_Height
   samplerSource=SSM_Wrap_WorldGroupSettings
   samplerType=SAMPLERTYPE_LinearGrayscale
   sortPriority=-5
   texture=&T_BlackLinearGrayscale
  }
  n30 = TextureSampleParameter2D(UVs=@n27) {
   group=Material_1
   parameterName=Background Diffuse
   samplerSource=SSM_Wrap_WorldGroupSettings
   sortPriority=-5
   texture=&T_WhitesRGB
  }
  n31 = TextureSampleParameter2D(UVs=@n28) {
   group=Material_1
   parameterName=Diffuse
   samplerSource=SSM_Wrap_WorldGroupSettings
   sortPriority=-5
   texture=&T_MidgreysRGB
  }
  n32 = StaticSwitchParameter(True=@n29.R, False=StaticSwitchParameter(True=@n30.A, False=@n31.A) {group=Material_1, parameterName=Use Color Swatch}) {desc=Use Smale scale grayscale, group=Material_1, parameterName=L1 Custom Height, sortPriority=1}
  n33 = MaterialFunctionCall!(
   Transition=ScalarParameter() {defaultValue=0.5, group=Material_1, parameterName=L1_TintTransition, sortPriority=5}
   HM=@n32
   Contrast=ScalarParameter() {defaultValue=5, group=Material_1, parameterName=L1_TintContrast, sortPriority=5}
   Metalness=@n17.G
   BG Color=StaticSwitchParameter(True=@n30.RGB, False=@n31.RGB) {group=Material_1, parameterName=Use Color Swatch}
   TintColor=StaticSwitchParameter(True=@n31.RGB, False=VectorParameter() {group=Material_1, parameterName=TintColor, sortPriority=7, defaultValue=<rgba 1 1 1 0>}.RGB) {group=Material_1, parameterName=Use Color Swatch}
  ) {materialFunction=/Game/Athena/WinterArt/Materials/MF_ColorMask}
  n34 = AppendVector(
   A=StaticSwitchParameter(True=@n33.LerpMix, False=@n33.LerpMix) {group=Material_1, parameterName=L1_UseTintColorAsDiffuse}
   B=Multiply(A=@n32, B=ScalarParameter() {defaultValue=1, group=Material_1, parameterName=L1 Alpha, sortPriority=1})
  )
  n35 = StaticSwitchParameter(True=LinearInterpolate(A=@n26.RGBA, B=@n34, Alpha=@n14), False=@n34) {group=Material_1, parameterName=TwoSided, sortPriority=18}
  n36 = SphereMask(
   A=Subtract(A=NamedRerouteUsage(), B=ScalarParameter() {defaultValue=-0.25, parameterName=ReactiveEmissive})
   B=ScalarParameter() {parameterName=ReactiveEmissiveSamplePos}
   Radius=ScalarParameter() {defaultValue=0.25, parameterName=ReactiveEmissiveRadius}
   Hardness=ScalarParameter() {parameterName=ReactiveEmissiveHardness}
  )
  n37 = TextureCoordinate()
  n38 = ScalarParameter() {defaultValue=2, parameterName=EmissiveNoiseTiling}
  n39 = TextureSample(UVs=Panner(Coordinate=Multiply(A=@n37, B=@n38)) {speedX=0.1, speedY=0.1}) {samplerType=SAMPLERTYPE_LinearColor, texture=/Game/Environments/Artemis/Props/Daisy/Daisy_RealityGlitch/Textures/T_Tendrils}
  n40 = Clamp(None=Multiply(A=ComponentMask(None=@n39.RGB) {r=true}) {constB=20})
  n41 = Subtract(A=@n36, B=Multiply(A=OneMinus(None=@n40), B=ScalarParameter() {defaultValue=0.75, parameterName=EmissiveNoiseStrength}))
  n42 = TextureSample(UVs=@n37) {samplerType=SAMPLERTYPE_LinearGrayscale, texture=/Game/Effects/Fort_Effects/Textures/T_DarkAndLight_Noise}
  n43 = Multiply(
   A=Multiply(A=Clamp(None=@n41), B=TextureSample(UVs=@n42.G) {texture=/Game/Environments/Artemis/Props/Daisy/Daisy_RealityGlitch/Textures/T_4colors}.RGB)
   B=ScalarParameter() {defaultValue=4, parameterName=EmissiveStrength}
  )
  n44 = ScalarParameter() {defaultValue=2, parameterName=EmissivePower}
  n45 = TextureSampleParameter2D(UVs=Multiply(A=@n8.Normals, B=@n7.UVs Layer 1)) {
   group=Material_1
   parameterName=Normals Side 2
   samplerSource=SSM_Wrap_WorldGroupSettings
   samplerType=SAMPLERTYPE_Normal
   sortPriority=3
   texture=/Engine/EngineMaterials/FlatNormal
  }
  n46 = TextureSampleParameter2D(UVs=Multiply(A=@n15.Normals, B=@n7.UVs Layer 1)) {
   group=Material_1
   parameterName=Normals
   samplerSource=SSM_Wrap_WorldGroupSettings
   samplerType=SAMPLERTYPE_Normal
   sortPriority=-5
   texture=&FlatNormal
  }
  n47 = ScalarParameter() {defaultValue=1, group=Material_1, parameterName=L1_BoostNormals, sortPriority=3}
  n48 = Multiply(A=@n46.RGB, B=AppendVector(A=AppendVector(A=@n47, B=@n47), B=Constant() {r=1}))
  n49 = SetMaterialAttributes!(
   MaterialAttributes
   Base Color=@n35
   Metallic=@n22.G
   Specular=@n22.G
   Roughness=@n22.G
   Normal=StaticSwitchParameter(True=LinearInterpolate(A=@n45.RGB, B=@n48, Alpha=@n14), False=@n48) {group=Material_1, parameterName=TwoSided, sortPriority=18}
   Emissive Color=Add(
    A=MaterialFunctionCall!(BaseColor=@n35, EmissiveMask=@n23.R, EmissiveColor=@n25.EmissiveMasked) {materialFunction=/Game/Athena/WinterArt/Materials/MF_EmissiveControls}.Result
    B=Power(Base=@n43, Exp=@n44)
   )
   World Position Offset
  ) {attributeSetTypes [
    #1
    #2
    #3
    #4
    #5
    #6
    #7
   ]}
  n50 = MaterialFunctionCall!(Material=MaterialFunctionCall!(Z-Gradient=NamedRerouteUsage(), Material=@n49) {materialFunction=&/MFC_SaplingGhost}.Result) {materialFunction=&/MFC_SaplingTeleport}
  n51 = RuntimeVirtualTextureSampleParameter() {
   group=RVT
   materialType=BaseColor_Normal_Specular_YCoCg
   parameterName=RVT
   sortPriority=5
   virtualTexture=/Game/Athena/Apollo/Maps/Landscape/VT/RVT_Apollo_Landscape
  }
  n52 = SetMaterialAttributes!(MaterialAttributes=@n50.Result, Base Color=@n51.BaseColor, Specular=@n51.BaseColor, Roughness=@n51.BaseColor, Normal=@n51.BaseColor) {attributeSetTypes [
    #1
    #3
    #4
    #5
   ]}
  n53 = ComponentMask(None=WorldPosition().XYZ) {b=true}
  n54 = RuntimeVirtualTextureSampleParameter() {
   group=RVT
   materialType=WorldHeight
   parameterName=RVTHeight
   sortPriority=5
   virtualTexture=/Game/Athena/Apollo/Maps/Landscape/VT/RVT_Apollo_Landscape_Height
  }
  n55 = MaterialFunctionCall(
   Transition Phase=Divide(A=Subtract(A=@n53, B=@n54.WorldHeight), B=ScalarParameter() {defaultValue=48, parameterName=RVT_BlendDist})
   Height Texture=@n42.G
  ) {materialFunction=&HeightLerp}
  n56 = MaterialFunctionCall!(
   IsStaticSnow
   UseAlternateStaticSnowMask
   AlternateStaticSnowMask
   "UseRoad&&OcclusionMask"
   IsForGrassMeshToPushDown
   AdditionalMaskToRemoveSnow
   IsFallingSnowFX
   UseMinimapCoords
   IncreaseSnowUsingSlope
   UseSmoothBorder
   LerpHighM
   TemporalHeightLerpContrast
   IsLobbyIsland
   UseBR-IslandMask=StaticBool()
   IsTree
   IsForRenderTarget
   UseSparkle
   UseSnowNormals=@n6
   UseAdditionalMaskToRemoveSnow
   IsTangentSpaceInput=@n6
   UseVertexNormalsForBasicSnow
   UseTangentSnowNormals=@n6
   Material=MaterialFunctionCall(NoRVT=@n50.Result, RVT=BlendMaterialAttributes(A=@n52, B=@n50.Result, Alpha=@n55.Lerp Alpha No Contrast) {vertexAttributeBlendType=UseA}) {materialFunction=&FN_RVT_Switch_Material}.Result
  ) {
   desc=Realtime
   materialFunction=/Game/Athena/Apollo/Environments/Landscape/Materials/Biome/UniqueBiomeMaterialFunctions/MF_Apollo_Snow_Progression
  }
  n57 = GetMaterialAttributes(None=MaterialFunctionCall!(In=@n56."DefaultSwitch=FalseMaterial") {materialFunction=/Game/Athena/WinterArt/Materials/MF_GameWideShared}.Result) {attributeGetTypes [#7]}
  n58 = MaterialFunctionCall!(MaskInput=@n36) {materialFunction=&/MFC_SaplingWPO}
  SetMaterialAttributes!(MaterialAttributes=@n57.MaterialAttributes, World Position Offset=Add(A=@n57.World Position Offset, B=Add(A=@n50.WPOResult, B=@n58.Result))) {attributeSetTypes [#7]}
  VectorParameter() {parameterName=ReactiveEmissiveColor, defaultValue=<rgba 1 0.5846 0 1>}
 ]
}
